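# Trim material function for the Gherkin skin of the Lime vehicle body, outputting material attributes.
# Pre-skinned local position drives two side-on planar projections of the trim texture, one flipped and offset, chosen by side.
# Static switches swap in chassis blending, body base color or PBR textures, and add mask-driven tinted emissive over a tiled noise normal.
# An optional clear coat branch adds a flattened bottom normal and blends shading models by coat thickness.
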MaterialFunction /VehicleCosmetics/Bodies/Lime/Skins/Gherkin/Trim/MF_Trim_Base_Gherkin
graph {nodes [
  n1 = Constant2Vector() {desc=UVTiling, g=0.25, r=0.25}
  n2 = VertexInterpolator(VS=Divide(A=PreSkinnedPosition()) {constB=100, desc=PSLP_Scaled}) {desc=PSLP_Scaled_VI}
  n3 = ComponentMask(None=@n2.PS) {r=true}
  n4 = ComponentMask(None=@n2.PS) {b=true}
  n5 = Multiply(A=@n1, B=AppendVector(A=@n3, B=OneMinus(None=@n4)))
  n6 = OneMinus(None=@n3)
  n7 = OneMinus(None=@n4)
  n8 = Add(A=Multiply(A=@n1, B=AppendVector(A=@n6, B=@n7)), B=Constant2Vector() {desc=LateralOffset, g=0.5, r=0.6})
  n9 = Constant() {desc=SideFeather}
  n10 = ComponentMask(None=@n2.PS) {g=true}
  n11 = LinearInterpolate(
   A=TextureSampleParameter2D(UVs=Frac(None=@n5)) {group=Trim, parameterName=Trim_Texture, texture=/Game/Environments/Sets/ArtDeco/Architecture/Textures/AD_Swatch_Grey}.RGB
   B=TextureSampleParameter2D(UVs=Frac(None=@n8)) {group=Trim, parameterName=Trim_Texture, texture=&AD_Swatch_Grey}.RGB
   Alpha=Subtract(B=Step(Y=@n9, X=@n10))
  )
  n12 = TextureSampleParameter2D() {parameterName=Chasis_Texture, texture=&AD_Swatch_Grey}
  n13 = LinearInterpolate(A=@n11, B=@n12.RGB, Alpha=@n12.RGB)
  n14 = TextureSampleParameter2D() {
   desc=_M
   group=Textures
   parameterName=Masks
   samplerSource=SSM_Wrap_WorldGroupSettings
   samplerType=SAMPLERTYPE_LinearColor
   sortPriority=33
   texture=/VehicleCosmetics/Textures/T_Black_Linear
   channelNames {a="", b=BoostMask, g=PaintMask, r=EmssivePartsMask}
  }
  n15 = TextureSampleParameter2D() {
   group=Texture
   parameterName=BodyTextures
   samplerSource=SSM_Wrap_WorldGroupSettings
   samplerType=SAMPLERTYPE_LinearColor
   texture=&T_Black_Linear
  }
  n16 = StaticSwitchParameter(True=@n14.EmssivePartsMask, False=@n15.B) {parameterName=ApplyTrim}
  n17 = Saturate(None=Multiply(A=Subtract(A=@n16) {constB=0.098}) {constB=1.1087})
  n18 = Multiply(
   A=VectorParameter() {group=Trim, parameterName=Trim_EmissiveColor, sortPriority=50, defaultValue=<rgba 0 1 1 1>}.RGB
   B=ScalarParameter() {defaultValue=1, group=Trim, parameterName=Trim_EmissiveBrightness, sortPriority=51}
  )
  n19 = SetMaterialAttributes!(
   MaterialAttributes
   BaseColor=StaticSwitchParameter(True=NamedRerouteUsage(), False=StaticSwitchParameter(True=@n13, False=@n11) {parameterName=Chasis_UseBaseColor_Texture}) {group=Trim, parameterName=Trim_UseBaseColorTexture}
   Metallic=StaticSwitchParameter(True=NamedRerouteUsage(), False=ScalarParameter() {group=Trim, parameterName=Trim_Metallic}) {group=Trim, parameterName=Trim_UseBodyTexturesPBR}
   Roughness=StaticSwitchParameter(True=NamedRerouteUsage(), False=ScalarParameter() {defaultValue=0.45, group=Trim, parameterName=Trim_Roughness}) {group=Trim, parameterName=Trim_UseBodyTexturesPBR}
   EmissiveColor=StaticSwitchParameter(True=Multiply(A=@n17, B=@n18), False=Constant()) {group=Trim, parameterName=ApplyTrimEmissive}
   Normal=NamedRerouteUsage()
   ShadingModel=ShadingModel()
  ) {attributeSetTypes [
    #1
    #2
    #3
    #4
    #5
    #6
   ]}
  n20 = Saturate(None=ScalarParameter() {defaultValue=1, group=Trim, parameterName=Trim_ClearCoat})
  n21 = TextureSampleParameter2D() {group=Finish, parameterName=TrimBottomNormalBase, samplerType=SAMPLERTYPE_Normal, texture=/VehicleCosmetics/Textures/T_BlankNormal_B}
  n22 = ScalarParameter() {group=DetailNormal, parameterName=TrimBottomNormalFlatness}
  n23 = SetMaterialAttributes!(
   MaterialAttributes=@n19
   CustomData0=@n20
   CustomData1=Saturate(None=ScalarParameter() {group=Trim, parameterName=Trim_ClearCoatRoughness})
   ClearCoatBottomNormal=NamedRerouteDeclaration(None=MaterialFunctionCall(Normal=@n21.RGB, Flatness=@n22) {materialFunction=&FlattenNormal}.Result) {name=CCBottomNormal, nodeColor=<rgba 1 0.6562 0 1>}
   ShadingModel=ShadingModel() {shadingModel=MSM_ClearCoat}
  ) {attributeSetTypes [
    #7
    #8
    #9
    #6
   ]}
  Composite!("In (MA)"=@n23, CCThickness=@n20) {subgraphName=BlendShadingModel}
  MaterialExpressionReroute_9 = Reroute(None=@n23)
  n24 = GetMaterialAttributes(None=@MaterialExpressionReroute_9."In (MA)") {attributeGetTypes [#5, #9]}
  MaterialExpressionReroute_10 = Reroute(None=@n20)
  n25 = BlendMaterialAttributes(
   A=SetMaterialAttributes!(
    MaterialAttributes=@n24.MaterialAttributes
    ShadingModel=ShadingModel()
    Normal=MaterialFunctionCall(BaseNormal=@n24.Normal, AdditionalNormal=@n24.Normal) {materialFunction=&BlendAngleCorrectedNormals}.Result
   ) {attributeSetTypes [#6, #5]}
   B=@MaterialExpressionReroute_9."In (MA)"
   Alpha=Saturate(None=Ceil(None=@MaterialExpressionReroute_10.CCThickness))
  )
  MaterialExpressionReroute_11 = Reroute(None=@n25)
  FunctionOutput(None=StaticSwitchParameter(True=@MaterialExpressionReroute_11.Output, False=@n19) {group=Trim, parameterName=Trim_UseClearCoatShading})
  MaterialFunctionCall() {materialFunction=&PlatformSwitch_1D}
  MaterialFunctionCall!("Xb1,PS4", "PC-High&&Epic,XSX,PS5", "PC-Perf,Android", Switch, "PC-Low(NotPerf)&&Medium") {materialFunction=/Game/Effects/Niagara/MasterMaterials/Functions/Common/PlatformSwitch_2D}
  n26 = TextureSampleParameter2D() {
   group=Textures
   parameterName=BaseColor
   samplerSource=SSM_Wrap_WorldGroupSettings
   sortPriority=30
   texture=/VehicleCosmetics/Textures/T_Black
  }
  NamedRerouteDeclaration(None=@n26.RGB) {name=BaseColorTexture, nodeColor=<rgba 0.6038 0.3325 0.013 1>}
  n27 = TextureSampleParameter2D() {
   group=Trim
   parameterName=BodyTextures
   samplerSource=SSM_Wrap_WorldGroupSettings
   samplerType=SAMPLERTYPE_LinearColor
   sortPriority=35
   texture=&T_Black_Linear
   channelNames {a=Roughness, b=Metallic, g=AO-BlackOut, r=Curvature}
  }
  NamedRerouteDeclaration(None=@n27.Roughness) {name=Roughness, nodeColor=<rgba 0 0.1726 0.4792 1>}
  NamedRerouteDeclaration(None=@n27.Metallic) {name=Metallic, nodeColor=<rgba 0 0.1726 0.4792 1>}
  n28 = TextureSampleParameter2D(UVs=Multiply(A=TextureCoordinate(), B=ScalarParameter() {defaultValue=64, group=Trim, parameterName=Trim_NoiseTiling})) {
   group=Textures
   parameterName=Normal
   samplerSource=SSM_Wrap_WorldGroupSettings
   samplerType=SAMPLERTYPE_Normal
   sortPriority=34
   texture=/VehicleCosmetics/Textures/UniversalTextures/NormalNoise_universal
  }
  NamedRerouteDeclaration(None=QualitySwitch(Default=@n28.RGB, Low=Constant3Vector() {constant=<rgba 0 0 1 0>})) {name=NormalTexture, nodeColor=<rgba 0.6038 0.3325 0.013 1>}
  OneMinus(None=@n9)
  PinBase() {reroutePins [
    "In (MA)" {expression=&self:MaterialExpressionReroute_9}
    CCThickness {expression=&self:MaterialExpressionReroute_10}
   ]}
  PinBase!(Output=@n25) {reroutePins [Output {expression=&self:MaterialExpressionReroute_11}]}
 ]}
