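# Base trim material for the Gherkin skin of the Lime vehicle body, supporting Default Lit and Clear Coat shading.
# MF_Trim_Base_Gherkin supplies the material attributes, which a static switch can blend toward shared carbon fiber by a scalar parameter.
# The result passes through MF_Vehicle_FNPostFX, while a named reroute carries the clear coat bottom normal separately.
# De-rez and damage function nodes are present, but their outputs are unused.
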Material /VehicleCosmetics/Bodies/Lime/Skins/Gherkin/Trim/MAT_Trim_Base_Gherkin
graph {
 properties {
  bUseMaterialAttributes=true
  bUsedWithMeshDeformer=true
  bUsedWithNanite=true
  bUsedWithSkeletalMesh=true
  pixelDepthOffsetMode=PDOM_Legacy
  shadingModel=MSM_FromMaterialExpression
  usedShadingModels=Default Lit | Clear Coat
 }
 nodes [
  ClearCoatNormalCustomOutput!(None=NamedRerouteUsage())
  n1 = GetMaterialAttributes(None=MaterialFunctionCall() {materialFunction=&/MF_Trim_Base_Gherkin}.Result) {attributeGetTypes [#1]}
  n2 = StaticSwitchParameter(
   True=BlendMaterialAttributes(
    A=@n1.MaterialAttributes
    B=MaterialFunctionCall() {materialFunction=/VehicleCosmetics/SharedMaterials/MF_CarbonFiber_VC}.Result
    Alpha=ScalarParameter() {group=SpecialFeatures, parameterName=Trim_SwapTrimForCarbonFiber}
   )
   False=@n1.MaterialAttributes
  ) {group=SpecialFeatures, parameterName=TrimSwapForCarbonFiber}
  MaterialFunctionCall(In=@n2) {materialFunction=&MF_VehicleDeRez_Transition}
  MaterialFunctionCall() {materialFunction=&MF_BaseVehicle_Damage}
  MaterialFunctionCall!(In=@n2) {materialFunction=/VehicleCosmetics/Materials/MaterialFunctions/MF_Vehicle_FNPostFX}
  NamedRerouteDeclaration(None=@n1.ClearCoatBottomNormal) {name=Trim_BottomNormal, nodeColor=<rgba 1 0.0234 0 1>}
 ]
}
